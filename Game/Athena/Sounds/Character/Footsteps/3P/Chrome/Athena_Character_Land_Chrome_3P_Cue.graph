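# Sound cue for third-person character landing footsteps on chrome in Athena, reusing first-person chrome landing waves under an enemy foley sound class.
# A quality-level node picks six random variations for its first two slots and four for the last two.
# A modulator varies pitch and volume, then feeds a mixer whose other input is empty.
# Third-person landing attenuation and a global footstep concurrency set govern playback.
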SoundCue /Game/Athena/Sounds/Character/Footsteps/3P/Chrome/Athena_Character_Land_Chrome_3P_Cue
root=Mixer_5
nodes [
 {id=Mixer_5, children [null,Modulator_5], properties {inputVolume [0.4,1]}}
 {id=Modulator_5, children [QualityLevel_0], properties {pitchMax=1.1, pitchMin=0.9, volumeMax=1.3, volumeMin=0.7}}
 {id=QualityLevel_0, children [Random_4,Random_4,Random_5,Random_5]}
 {
  id=Random_4
  children [WavePlayer_5, WavePlayer_8, WavePlayer_9, WavePlayer_10, WavePlayer_11, WavePlayer_12]
  properties {weights [1,1,1,1,1,1]}
 }
 {
  id=Random_5
  children [WavePlayer_13, WavePlayer_14, WavePlayer_15, WavePlayer_16]
  properties {bShouldExcludeFromBranchCulling=true, weights [1,1,1,1]}
 }
 {id=WavePlayer_10, properties {soundWaveAssetPtr=/Game/Athena/Sounds/Character/Footsteps/1P/Chrome/Foley_FS_1P_Chrome_Land_04}}
 {id=WavePlayer_11, properties {soundWaveAssetPtr=/Game/Athena/Sounds/Character/Footsteps/1P/Chrome/Foley_FS_1P_Chrome_Land_05}}
 {id=WavePlayer_12, properties {soundWaveAssetPtr=/Game/Athena/Sounds/Character/Footsteps/1P/Chrome/Foley_FS_1P_Chrome_Land_06}}
 {id=WavePlayer_13, properties {soundWaveAssetPtr=/Game/Athena/Sounds/Character/Footsteps/1P/Chrome/Foley_FS_1P_Chrome_Land_01}}
 {id=WavePlayer_14, properties {soundWaveAssetPtr=/Game/Athena/Sounds/Character/Footsteps/1P/Chrome/Foley_FS_1P_Chrome_Land_02}}
 {id=WavePlayer_15, properties {soundWaveAssetPtr=/Game/Athena/Sounds/Character/Footsteps/1P/Chrome/Foley_FS_1P_Chrome_Land_03}}
 {id=WavePlayer_16, properties {soundWaveAssetPtr=&Foley_FS_1P_Chrome_Land_04}}
 {id=WavePlayer_5, properties {soundWaveAssetPtr=&Foley_FS_1P_Chrome_Land_01}}
 {id=WavePlayer_8, properties {soundWaveAssetPtr=&Foley_FS_1P_Chrome_Land_02}}
 {id=WavePlayer_9, properties {soundWaveAssetPtr=&Foley_FS_1P_Chrome_Land_03}}
]
properties {
 attenuationSettings=/Game/Sounds/Attenuation/Fort_Player_Land_3P_Athena_Att
 maxDistance=3584
 soundClassObject=&Fort_Player_Foley_3P_Enemy_Athena
 volumeMultiplier=0.9
 concurrencyOverrides {bLimitToOwner=true, maxCount=4, resolutionRule=StopOldest}
 concurrencySet [&Athena_Footstep_Global_Conc]
}
