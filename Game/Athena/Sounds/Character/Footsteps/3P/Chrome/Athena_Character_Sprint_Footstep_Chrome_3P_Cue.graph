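# Sound cue for third-person sprint footsteps on chrome, routed to the enemy foley sound class.
# A quality switch feeds its first two levels from a ten-wave random pool and its last two from a five-wave subset.
# The waves are first-person chrome run recordings, varied in volume and pitched up by a root modulator.
# Sprint footstep attenuation and local and global concurrency sets govern playback.
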SoundCue /Game/Athena/Sounds/Character/Footsteps/3P/Chrome/Athena_Character_Sprint_Footstep_Chrome_3P_Cue
root=Modulator_2
nodes [
 {id=Modulator_2, children [QualityLevel_0], properties {pitchMax=1.5, pitchMin=1.3, volumeMax=1, volumeMin=0.7}}
 {id=QualityLevel_0, children [Random_3,Random_3,Random_4,Random_4]}
 {
  id=Random_3
  children [
   WavePlayer_25
   WavePlayer_26
   WavePlayer_27
   WavePlayer_28
   WavePlayer_29
   WavePlayer_30
   WavePlayer_31
   WavePlayer_32
   WavePlayer_33
   WavePlayer_34
  ]
  properties {weights [1,1,1,1,1,1,1,1,1,1]}
 }
 {id=Random_4, children [WavePlayer_36, WavePlayer_37, WavePlayer_38, WavePlayer_39, WavePlayer_40], properties {weights [1,1,1,1,1]}}
 {id=WavePlayer_25, properties {soundWaveAssetPtr=/Game/Athena/Sounds/Character/Footsteps/1P/Chrome/Foley_FS_1P_Chrome_Run_10}}
 {id=WavePlayer_26, properties {soundWaveAssetPtr=/Game/Athena/Sounds/Character/Footsteps/1P/Chrome/Foley_FS_1P_Chrome_Run_01}}
 {id=WavePlayer_27, properties {soundWaveAssetPtr=/Game/Athena/Sounds/Character/Footsteps/1P/Chrome/Foley_FS_1P_Chrome_Run_02}}
 {id=WavePlayer_28, properties {soundWaveAssetPtr=/Game/Athena/Sounds/Character/Footsteps/1P/Chrome/Foley_FS_1P_Chrome_Run_03}}
 {id=WavePlayer_29, properties {soundWaveAssetPtr=/Game/Athena/Sounds/Character/Footsteps/1P/Chrome/Foley_FS_1P_Chrome_Run_04}}
 {id=WavePlayer_30, properties {soundWaveAssetPtr=/Game/Athena/Sounds/Character/Footsteps/1P/Chrome/Foley_FS_1P_Chrome_Run_05}}
 {id=WavePlayer_31, properties {soundWaveAssetPtr=/Game/Athena/Sounds/Character/Footsteps/1P/Chrome/Foley_FS_1P_Chrome_Run_06}}
 {id=WavePlayer_32, properties {soundWaveAssetPtr=/Game/Athena/Sounds/Character/Footsteps/1P/Chrome/Foley_FS_1P_Chrome_Run_07}}
 {id=WavePlayer_33, properties {soundWaveAssetPtr=/Game/Athena/Sounds/Character/Footsteps/1P/Chrome/Foley_FS_1P_Chrome_Run_08}}
 {id=WavePlayer_34, properties {soundWaveAssetPtr=/Game/Athena/Sounds/Character/Footsteps/1P/Chrome/Foley_FS_1P_Chrome_Run_09}}
 {id=WavePlayer_36, properties {soundWaveAssetPtr=&Foley_FS_1P_Chrome_Run_01}}
 {id=WavePlayer_37, properties {soundWaveAssetPtr=&Foley_FS_1P_Chrome_Run_02}}
 {id=WavePlayer_38, properties {soundWaveAssetPtr=&Foley_FS_1P_Chrome_Run_03}}
 {id=WavePlayer_39, properties {soundWaveAssetPtr=&Foley_FS_1P_Chrome_Run_04}}
 {id=WavePlayer_40, properties {soundWaveAssetPtr=&Foley_FS_1P_Chrome_Run_05}}
]
properties {
 attenuationSettings=/Game/Sounds/Attenuation/Fort_Player_Footsteps_3P_Sprint_Athena_Att
 maxDistance=3584
 soundClassObject=&Fort_Player_Foley_3P_Enemy_Athena
 volumeMultiplier=1.2
 concurrencyOverrides {bLimitToOwner=true, maxCount=2, resolutionRule=StopOldest}
 concurrencySet [&Athena_Footstep_Local_Conc, &Athena_Footstep_Global_Conc]
}
