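# Third-person walking footstep cue for chrome surfaces in Athena, using 3P enemy foley sound class, walk attenuation and owner-limited concurrency.
# Randomized chrome step waves, reused from 1P recordings, mix with Commando foley, then pass through pitch-raising, volume-varying modulation.
# A quality switch gives the first two levels larger pools with very quiet foley, and the last two smaller pools with full-volume foley.
# The root mixer's first input is empty.
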SoundCue /Game/Athena/Sounds/Character/Footsteps/3P/Chrome/Athena_Character_Walk_Footstep_Chrome_3P_Cue
root=Mixer_1
nodes [
 {id=Mixer_1, children [null,Modulator_2], properties {inputVolume [0.1,1]}}
 {id=Mixer_2, children [Random_5,Random_2], properties {inputVolume [0.05,1]}}
 {id=Mixer_3, children [Random_0,Random_4], properties {inputVolume [1,1]}}
 {id=Modulator_2, children [QualityLevel_0], properties {pitchMax=1.5, pitchMin=1.3, volumeMax=1, volumeMin=0.7}}
 {id=QualityLevel_0, children [Mixer_2,Mixer_2,Mixer_3,Mixer_3]}
 {id=Random_0, children [WavePlayer_12, WavePlayer_13, WavePlayer_14], properties {bShouldExcludeFromBranchCulling=true, weights [1,1,1]}}
 {
  id=Random_2
  children [
   WavePlayer_3
   WavePlayer_4
   WavePlayer_5
   WavePlayer_15
   WavePlayer_16
   WavePlayer_17
   WavePlayer_24
   WavePlayer_25
   WavePlayer_26
   WavePlayer_27
  ]
  properties {weights [1,1,1,1,1,1,1,1,1,1]}
 }
 {
  id=Random_4
  children [WavePlayer_28, WavePlayer_29, WavePlayer_30, WavePlayer_31, WavePlayer_32]
  properties {bShouldExcludeFromBranchCulling=true, weights [1,1,1,1,1]}
 }
 {
  id=Random_5
  children [WavePlayer_18, WavePlayer_6, WavePlayer_7, WavePlayer_11, WavePlayer_9, WavePlayer_10]
  properties {bShouldExcludeFromBranchCulling=true, weights [1,1,1,1,1,1]}
 }
 {id=WavePlayer_10, properties {soundWaveAssetPtr=/Game/Sounds/Fortress_Foley/Commando/Commando_Foley_Stereo_008}}
 {id=WavePlayer_11, properties {soundWaveAssetPtr=/Game/Sounds/Fortress_Foley/Commando/Commando_Foley_Stereo_009}}
 {id=WavePlayer_12, properties {soundWaveAssetPtr=/Game/Sounds/Fortress_Foley/Commando/Commando_Foley_Stereo_001}}
 {id=WavePlayer_13, properties {soundWaveAssetPtr=/Game/Sounds/Fortress_Foley/Commando/Commando_Foley_Stereo_003}}
 {id=WavePlayer_14, properties {soundWaveAssetPtr=/Game/Sounds/Fortress_Foley/Commando/Commando_Foley_Stereo_005}}
 {id=WavePlayer_15, properties {soundWaveAssetPtr=/Game/Athena/Sounds/Character/Footsteps/1P/Chrome/Foley_FS_1P_Chrome_Walk_04}}
 {id=WavePlayer_16, properties {soundWaveAssetPtr=/Game/Athena/Sounds/Character/Footsteps/1P/Chrome/Foley_FS_1P_Chrome_Walk_05}}
 {id=WavePlayer_17, properties {soundWaveAssetPtr=/Game/Athena/Sounds/Character/Footsteps/1P/Chrome/Foley_FS_1P_Chrome_Walk_06}}
 {id=WavePlayer_18, properties {soundWaveAssetPtr=&Commando_Foley_Stereo_001}}
 {id=WavePlayer_24, properties {soundWaveAssetPtr=/Game/Athena/Sounds/Character/Footsteps/1P/Chrome/Foley_FS_1P_Chrome_Walk_07}}
 {id=WavePlayer_25, properties {soundWaveAssetPtr=/Game/Athena/Sounds/Character/Footsteps/1P/Chrome/Foley_FS_1P_Chrome_Walk_08}}
 {id=WavePlayer_26, properties {soundWaveAssetPtr=/Game/Athena/Sounds/Character/Footsteps/1P/Chrome/Foley_FS_1P_Chrome_Walk_09}}
 {id=WavePlayer_27, properties {soundWaveAssetPtr=/Game/Athena/Sounds/Character/Footsteps/1P/Chrome/Foley_FS_1P_Chrome_Walk_10}}
 {id=WavePlayer_28, properties {soundWaveAssetPtr=/Game/Athena/Sounds/Character/Footsteps/1P/Chrome/Foley_FS_1P_Chrome_Walk_01}}
 {id=WavePlayer_29, properties {soundWaveAssetPtr=/Game/Athena/Sounds/Character/Footsteps/1P/Chrome/Foley_FS_1P_Chrome_Walk_02}}
 {id=WavePlayer_3, properties {soundWaveAssetPtr=&Foley_FS_1P_Chrome_Walk_01}}
 {id=WavePlayer_30, properties {soundWaveAssetPtr=/Game/Athena/Sounds/Character/Footsteps/1P/Chrome/Foley_FS_1P_Chrome_Walk_03}}
 {id=WavePlayer_31, properties {soundWaveAssetPtr=&Foley_FS_1P_Chrome_Walk_04}}
 {id=WavePlayer_32, properties {soundWaveAssetPtr=&Foley_FS_1P_Chrome_Walk_05}}
 {id=WavePlayer_4, properties {soundWaveAssetPtr=&Foley_FS_1P_Chrome_Walk_02}}
 {id=WavePlayer_5, properties {soundWaveAssetPtr=&Foley_FS_1P_Chrome_Walk_03}}
 {id=WavePlayer_6, properties {soundWaveAssetPtr=&Commando_Foley_Stereo_003}}
 {id=WavePlayer_7, properties {soundWaveAssetPtr=&Commando_Foley_Stereo_005}}
 {id=WavePlayer_9, properties {soundWaveAssetPtr=/Game/Sounds/Fortress_Foley/Commando/Commando_Foley_Stereo_007}}
]
properties {
 attenuationSettings=/Game/Sounds/Attenuation/Fort_Player_Footsteps_3P_Walk_Athena_Att
 maxDistance=3584
 soundClassObject=&Fort_Player_Foley_3P_Enemy_Athena
 volumeMultiplier=0.8
 concurrencyOverrides {bLimitToOwner=true, maxCount=2, resolutionRule=StopOldest}
 concurrencySet [&Athena_Footstep_Local_Conc, &Athena_Footstep_Global_Conc]
}
